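; Parent quest prop actor for signal jammers that players damage to progress a Battle Pass S24 narrative quest.
; On server damage, it checks that the instigating player has the prop's quest and has not completed its objective.
; If eligible, it applies the on-complete gameplay effect to that player's pawn and calls the damage FX event.
; That event plays Niagara FX and a destroyed sound at the mesh only without authority, gated by a local-controller check.
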(event EventOnDamageServer
  (params (Damage :type Float%20%28single-precision%29) (DamageTags :type Gameplay%20Tag%20Container%20Structure%20%28by%20ref%29) (Momentum :type Vector :default "0, 0, 0") (HitInfo :type Hit%20Result%20Structure%20%28by%20ref%29) (InstigatedBy :type Controller%20Object%20Reference) (DamageCauser :type Actor%20Object%20Reference) (EffectContext :type Gameplay%20Effect%20Context%20Handle%20Structure))
  ;[0 0 0] When signal jammer is hit, confirm player has quest and objective, then play the sound and FX
  (|CallParentFunction Damage DamageTags Momentum HitInfo InstigatedBy DamageCauser EffectContext)
  (bind _asfort_player_controller (Utilities|Casting|CastToFortPlayerController InstigatedBy))
  (bind _bcaninteract (CallFunction|InstigatorCanInteract _asfort_player_controller))
  (if _bcaninteract
    (GameplayEffects|ApplyGameplayEffectToSelf :self (Ability|GetFortAbilitySystemComponent (Game|Player|GetPlayerPawn _asfort_player_controller)) :GameplayEffectClass (Variables|InstanceGE|GetGameplayEffectOnComplete) :Level 0.0)
    (CallFunction|SpawnDamageFXAndAudio _asfort_player_controller)))

(event Custom|SpawnDamageFXAndAudio
  (params (InstigatingPlayerController :type Fort%20Player%20Controller%20Object%20Reference))
  (bind _sm_quest_preinteraction (Variables|Components|GetSM_Quest_PreInteraction))
  (bind _returnvalue (Transformation|GetWorldLocation _sm_quest_preinteraction))
  ;[0 0 0] Play sound and FX. Set up here
  (Networking|SwitchHasAuthority
    (:Remote
      (bind _output (CallFunction|IsValidLocalPlayerController InstigatingPlayerController))
      (if _output
        (Niagara|SpawnSystematLocation :SystemTemplate (Variables|AudioandFX|GetNS_DamageFX) :Location _returnvalue :Scale "1.000000,1.000000,1.000000" :bAutoDestroy false :bAutoActivate true :PoolingMethod "None" :bPreCullCheck true)
        (Audio|PlaySoundatLocation (Variables|AudioAndFX|GetDestroyedAudioCue) _returnvalue)))))

(fn InstigatorCanInteract
  (params (InstigatingPlayerController :type Fort%20Player%20Controller%20Object%20Reference))
  (bind _consolidatedquestcomponent (Variables|BuildingGameplayActorPropQuest|GetConsolidatedQuestComponent))
  (bind _returnvalue (Class|ConsolidatedQuestComponent|GetQuestItemDefinition _consolidatedquestcomponent))
  ;[0 0 0] Check if this player is on the quest and they have not already completed the objective for this prop instance
  (Utilities|IsValid InstigatingPlayerController
    (:Is%20Valid
      (Utilities|IsValid _returnvalue
        (:Is%20Valid
          (Utilities|IsValid (Quests|GetQuestManager InstigatingPlayerController "Invalid")
            (:Is%20Valid
              (bind _playerhasquest (Class|AthenaQuestFunctionLibrary|AthenaHasQuest InstigatingPlayerController _returnvalue))
              (if _playerhasquest
                (bind (_playercontrollerout _completedquest) (Quest|HasPlayerCompletedQuest InstigatingPlayerController _returnvalue (Variables|/Script/FortniteGame.ConsolidatedQuestComponent|GetObjectiveBackendName (Variables|BuildingGameplayActorPropQuest|GetConsolidatedQuestComponent))))
                (if (not _completedquest)
                  (return true)
                  (else
                    _))
                (else
                  _))
              (return false))
            (:Is%20Not%20Valid
              _)))
        (:Is%20Not%20Valid
          _)))
    (:Is%20Not%20Valid
      _)))

(fn IsValidLocalPlayerController
  (params (InstigatingPlayerController :type Fort%20Player%20Controller%20Object%20Reference))
  ;[0 0 0] Returns true on the executing client if the instigator controller is locally controlled. Returns false if running on the server
  (if (not (Networking|IsServer))
    (Utilities|IsValid InstigatingPlayerController
      (:Is%20Valid
        (if (PlayerController|IsinSplitscreen InstigatingPlayerController)
          (for _array_element (Fortnite|GetLocalFortPlayerControllers true)
            (if (== _array_element InstigatingPlayerController)
              (return true)))
          (else
            _))
        (return true))
      (:Is%20Not%20Valid
        _)))
  (return false))

(fn ConstructionScript
  (|CallParentFunction))
